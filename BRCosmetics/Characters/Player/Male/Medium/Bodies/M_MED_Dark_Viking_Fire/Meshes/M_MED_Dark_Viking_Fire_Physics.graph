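# Physics asset for the Dark Viking Fire male medium player body cosmetic, built around the rear loincloth's dynamic bones.
# Capsule bodies cover the pelvis, both thighs, a small loincloth driver and three loincloth segments.
# The thighs and driver attach to the pelvis with free angular motion.
# The segments chain from the driver with tight twist and moderate swing limits, their capsules offset from the bones.
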PhysicsAsset /BRCosmetics/Characters/Player/Male/Medium/Bodies/M_MED_Dark_Viking_Fire/Meshes/M_MED_Dark_Viking_Fire_Physics
bodies {
 dyn_rear_loincloth_01 [{shapeType=Capsule, length=15.25, radius=8.5, center=<10 -8.5 0>}]
 dyn_rear_loincloth_02 [{shapeType=Capsule, length=5, radius=10, center=<7 -9.5 0>}]
 dyn_rear_loincloth_03 [{shapeType=Capsule, length=10, radius=10, center=<6 -9.5 0>, rotation=<rot 90 0 0>}]
 dyn_rear_loincloth_driver [{shapeType=Capsule, length=1, radius=1}]
 pelvis [{shapeType=Capsule, length=12.5, radius=10}]
 thigh_l [{shapeType=Capsule, length=31, radius=8.5, center=<-22.5781 -0.0002 0>, rotation=<rot 90 0 0>}]
 thigh_r [{shapeType=Capsule, length=31, radius=8.5, center=<21.9044 -0.0003 0>, rotation=<rot 90 0 0>}]
}
constraints [
 {
  bone1Name=thigh_l
  bone2Name=pelvis
  swing1LimitDegrees=45
  swing1Motion=Free
  swing2LimitDegrees=45
  swing2Motion=Free
  twistLimitDegrees=45
  twistMotion=Free
 }
 {
  bone1Name=thigh_r
  bone2Name=pelvis
  swing1LimitDegrees=45
  swing1Motion=Free
  swing2LimitDegrees=45
  swing2Motion=Free
  twistLimitDegrees=45
  twistMotion=Free
 }
 {
  bone1Name=dyn_rear_loincloth_driver
  bone2Name=pelvis
  swing1LimitDegrees=45
  swing1Motion=Free
  swing2LimitDegrees=45
  swing2Motion=Free
  twistLimitDegrees=45
  twistMotion=Free
 }
 {bone1Name=dyn_rear_loincloth_01, bone2Name=dyn_rear_loincloth_driver, swing1LimitDegrees=15, swing2LimitDegrees=25, twistLimitDegrees=3}
 {bone1Name=dyn_rear_loincloth_02, bone2Name=dyn_rear_loincloth_01, swing1LimitDegrees=20, swing2LimitDegrees=20, twistLimitDegrees=3}
 {bone1Name=dyn_rear_loincloth_03, bone2Name=dyn_rear_loincloth_02, swing1LimitDegrees=20, swing2LimitDegrees=20, twistLimitDegrees=3}
]
