AnimBlueprint /BRCosmetics/Characters/Player/Male/Medium/Bodies/M_MED_Dark_Viking_Fire/Meshes/Parts/M_MED_Dark_Viking_Fire_FaceAcc_AnimBP
target_skeleton=&/M_MED_Dark_Viking_Fire_FaceAcc_Skeleton
graphs [
 AnimGraph {nodes [
   Anim Dynamics - Bone: dyn_neck_hair(Alpha=1, ExternalForce="0,0,0") {
    angularDampingOverride=-99
    angularSpringConstant=200
    bGravityOverrideInSimSpace=true
    bLinearSpring=true
    bOverrideAngularDamping=true
    bOverrideLinearDamping=true
    gravityScale=0.1
    lODThreshold=1
    linearDampingOverride=0.995
    linearSpringConstant=200
    alphaBoolBlend {blendOption=Linear}
    boundBone {boneName=dyn_neck_hair}
    componentLinearAccScale=<1 1 1>
    componentLinearVelScale=<1 1 1>
    physicsBodyDefinitions [{
      sphereCollisionRadius=0
      boundBone {boneName=dyn_neck_hair}
      boxExtents=<2 2 2>
      constraintSetup {angularLimitsMax=<0 0 10>, angularLimitsMin=<0 0 -10>, angularTarget=<0 0 0>, linearAxesMax=<0 0 2>, linearAxesMin=<0 0 -2>}
      localJointOffset=<0 0 -1>
     }]
   }
   "Constraint - dyn_beardBraid_02_c (PPPPPPPPP)"!(
    ComponentPose
    Alpha=1
    ConstraintWeights_0=0.2358
    ConstraintWeights_1:float=0.0229
    ConstraintWeights_2:float=0.6053
    ConstraintWeights_3:float=0.004
    ConstraintWeights_4:float=0.0222
    ConstraintWeights_5:float=0.059
    ConstraintWeights_6:float=0.0306
    ConstraintWeights_7:float=0.0202
    ConstraintWeights_8:float=0
   ) -> (Pose) {
    alphaBoolBlend {blendOption=Linear}
    boneToModify {boneName=dyn_beardBraid_02_c}
    constraintSetup [
     {targetBone {boneName=dyn_beard_01}}
     {targetBone {boneName=clavicle_l}}
     {targetBone {boneName=head_offset}}
     {targetBone {boneName=spine_02}}
     {targetBone {boneName=spine_03}}
     {targetBone {boneName=spine_04}}
     {targetBone {boneName=spine_05}}
     {targetBone {boneName=clavicle_r}}
     {targetBone {boneName=neck_01}}
    ]
    constraintWeights [1,1,1,1,1,1,1,1,1]
   }
   n1 = Get look_up() -> (look_up:double)
   n2 = "Transform (Modify) Bone - Bone: head_offset"(
    ComponentPose=Local To Component(LocalPose=Copy Pose From Mesh() {bUseAttachedParent=true})
    Alpha=@n1
    Translation="0,0,0"
    Rotation="0,-10,0"
    Scale="1,1,1"
   ) {
    rotationMode=BMM_Additive
    rotationSpace=BCS_BoneSpace
    alphaBoolBlend {blendOption=Linear}
    boneToModify {boneName=head_offset}
    rotation {yaw=-10}
   }
   n3 = "Transform (Modify) Bone - Bone: dyn_hair_bck_01"(ComponentPose=@n2, Alpha=@n1, Translation="0,-2.5,0", Rotation="0,-15,0", Scale="1,1,1") {
    rotationMode=BMM_Additive
    rotationSpace=BCS_BoneSpace
    translationMode=BMM_Additive
    translationSpace=BCS_BoneSpace
    alphaBoolBlend {blendOption=Linear}
    boneToModify {boneName=dyn_hair_bck_01}
    rotation {yaw=-15}
    translation {y=-2.5}
   }
   n4 = "Transform (Modify) Bone - Bone: dyn_hair_bck_02"(ComponentPose=@n3, Alpha=@n1, Translation="0,0,0", Rotation="0,10,0", Scale="1,1,1") {
    rotationMode=BMM_Additive
    rotationSpace=BCS_BoneSpace
    translationMode=BMM_Additive
    translationSpace=BCS_BoneSpace
    alphaBoolBlend {blendOption=Linear}
    boneToModify {boneName=dyn_hair_bck_02}
    rotation {yaw=10}
   }
   n5 = Get look_down() -> (look_down:double)
   n6 = "Transform (Modify) Bone - Bone: head_offset"(ComponentPose=@n4, Alpha=@n5, Translation="0,0,0", Rotation="0,15,0", Scale="1,1,1") {
    rotationMode=BMM_Additive
    rotationSpace=BCS_BoneSpace
    alphaBoolBlend {blendOption=Linear}
    boneToModify {boneName=head_offset}
    rotation {yaw=15}
   }
   n7 = "Transform (Modify) Bone - Bone: dyn_hair_bck_01"(ComponentPose=@n6, Alpha=@n5, Translation="0,0,0", Rotation="0,5,0", Scale="1,1,1") {
    rotationMode=BMM_Additive
    rotationSpace=BCS_BoneSpace
    translationMode=BMM_Additive
    translationSpace=BCS_BoneSpace
    alphaBoolBlend {blendOption=Linear}
    alphaScaleBiasClamp {bInterpResult=true, interpSpeedDecreasing=15, interpSpeedIncreasing=15}
    boneToModify {boneName=dyn_hair_bck_01}
    rotation {yaw=5}
   }
   n8 = "Transform (Modify) Bone - Bone: dyn_hair_bck_02"(ComponentPose=@n7, Alpha=@n5, Translation="0,-1,0", Rotation="0,8,0", Scale="1,1,1") {
    rotationMode=BMM_Additive
    rotationSpace=BCS_BoneSpace
    translationMode=BMM_Additive
    translationSpace=BCS_BoneSpace
    alphaBoolBlend {blendOption=Linear}
    alphaScaleBiasClamp {bInterpResult=true, interpSpeedDecreasing=15, interpSpeedIncreasing=15}
    boneToModify {boneName=dyn_hair_bck_02}
    rotation {yaw=8}
    translation {y=-1}
   }
   n9 = Get look_left() -> (look_left:double)
   n10 = "Transform (Modify) Bone - Bone: head_offset"(ComponentPose=@n8, Alpha=@n9, Translation="0,0,0", Rotation="0,0,25", Scale="1,1,1") {
    rotationMode=BMM_Additive
    rotationSpace=BCS_BoneSpace
    alphaBoolBlend {blendOption=Linear}
    boneToModify {boneName=head_offset}
    rotation {roll=25}
   }
   n11 = "Transform (Modify) Bone - Bone: dyn_hair_bck_01"(ComponentPose=@n10, Alpha=@n9, Translation="0,0,0", Rotation="0,1,0", Scale="1,1,1") {
    rotationMode=BMM_Additive
    rotationSpace=BCS_BoneSpace
    translationMode=BMM_Additive
    translationSpace=BCS_BoneSpace
    alphaBoolBlend {blendOption=Linear}
    alphaScaleBiasClamp {bInterpResult=true, interpSpeedDecreasing=15, interpSpeedIncreasing=15}
    boneToModify {boneName=dyn_hair_bck_01}
    rotation {yaw=1}
   }
   n12 = "Transform (Modify) Bone - Bone: dyn_hair_bck_01"(ComponentPose=@n11, Alpha=@n9, Translation="0,-1,0", Rotation="0,2,0", Scale="1,1,1") {
    rotationMode=BMM_Additive
    rotationSpace=BCS_BoneSpace
    translationMode=BMM_Additive
    translationSpace=BCS_BoneSpace
    alphaBoolBlend {blendOption=Linear}
    alphaScaleBiasClamp {bInterpResult=true, interpSpeedDecreasing=15, interpSpeedIncreasing=15}
    boneToModify {boneName=dyn_hair_bck_01}
    rotation {yaw=2}
    translation {y=-1}
   }
   n13 = Get look_right() -> (look_right:double)
   n14 = "Transform (Modify) Bone - Bone: head_offset"(ComponentPose=@n12, Alpha=@n13, Translation="0,0,0", Rotation="0,0,-25", Scale="1,1,1") {
    rotationMode=BMM_Additive
    rotationSpace=BCS_BoneSpace
    alphaBoolBlend {blendOption=Linear}
    boneToModify {boneName=head_offset}
    rotation {roll=-25}
   }
   n15 = "Transform (Modify) Bone - Bone: dyn_hair_bck_01"(ComponentPose=@n14, Alpha=@n13, Translation="0,0,2.5", Rotation="0,1,0", Scale="1,1,1") {
    rotationMode=BMM_Additive
    rotationSpace=BCS_BoneSpace
    translationMode=BMM_Additive
    translationSpace=BCS_BoneSpace
    alphaBoolBlend {blendOption=Linear}
    alphaScaleBiasClamp {bInterpResult=true, interpSpeedDecreasing=15, interpSpeedIncreasing=15}
    boneToModify {boneName=dyn_hair_bck_01}
    rotation {yaw=1}
    translation {z=2.5}
   }
   n16 = Anim Dynamics - Bone: dyn_beard_01(ComponentPose=@n15, Alpha=0.75, ExternalForce="0,0,0") {
    bGravityOverrideInSimSpace=true
    bLinearSpring=true
    bOverrideLinearDamping=true
    gravityScale=0.1
    lODThreshold=1
    linearDampingOverride=0.995
    linearSpringConstant=200
    alphaBoolBlend {blendOption=Linear}
    boundBone {boneName=dyn_beard_01}
    componentLinearAccScale=<1 1 1>
    componentLinearVelScale=<1 1 1>
    physicsBodyDefinitions [{
      sphereCollisionRadius=0
      boundBone {boneName=dyn_beard_01}
      boxExtents=<2 2 2>
      constraintSetup {angularLimitsMax=<0 0 0>, angularLimitsMin=<0 0 0>, angularTarget=<0 0 0>, linearAxesMax=<0 0 2>, linearAxesMin=<0 0 -2>}
      localJointOffset=<0 0 -1>
     }]
   }
   n17 = "Constraint - dyn_beardBraid_02_c (PPPPPPP)"!(
    ComponentPose=@n16
    Alpha=1
    ConstraintWeights_0=0.2359
    ConstraintWeights_1:float=0.0249
    ConstraintWeights_2:float=0.5956
    ConstraintWeights_3:float=0.0244
    ConstraintWeights_4:float=0.0645
    ConstraintWeights_5:float=0.0327
    ConstraintWeights_6:float=0.0221
   ) -> (Pose) {
    alphaBoolBlend {blendOption=Linear}
    boneToModify {boneName=dyn_beardBraid_02_c}
    constraintSetup [
     {targetBone {boneName=dyn_beard_01}}
     {targetBone {boneName=clavicle_l}}
     {targetBone {boneName=head_offset}}
     {targetBone {boneName=spine_03}}
     {targetBone {boneName=spine_04}}
     {targetBone {boneName=spine_05}}
     {targetBone {boneName=clavicle_r}}
    ]
    constraintWeights [1,1,1,1,1,1,1]
   }
   n18 = "Constraint - dyn_beardBraid_01_l (PPPPPPPPP)"!(
    ComponentPose=@n17
    Alpha=0.5
    ConstraintWeights_0=0.9995
    ConstraintWeights_1:float=0.0002
    ConstraintWeights_2:float=0.0001
    ConstraintWeights_3:float=0.0001
    ConstraintWeights_4:float=0
    ConstraintWeights_5:float=0
    ConstraintWeights_6:float=0
    ConstraintWeights_7:float=0
    ConstraintWeights_8:float=0
   ) -> (Pose) {
    alphaBoolBlend {blendOption=Linear}
    boneToModify {boneName=dyn_beardBraid_01_l}
    constraintSetup [
     {targetBone {boneName=head_offset}}
     {targetBone {boneName=dyn_beard_01}}
     {targetBone {boneName=clavicle_l}}
     {targetBone {boneName=spine_05}}
     {targetBone {boneName=spine_04}}
     {targetBone {boneName=spine_03}}
     {targetBone {boneName=spine_02}}
     {targetBone {boneName=clavicle_r}}
     {targetBone {boneName=neck_01}}
    ]
    constraintWeights [1,1,1,1,1,1,1,1,1]
   }
   n19 = "Constraint - dyn_beardBraid_01_r (PPPPPPPPP)"!(
    ComponentPose=@n18
    Alpha=0.5
    ConstraintWeights_0=0.9874
    ConstraintWeights_1:float=0.0052
    ConstraintWeights_2:float=0.0034
    ConstraintWeights_3:float=0.0032
    ConstraintWeights_4:float=0.0007
    ConstraintWeights_5:float=0.0001
    ConstraintWeights_6:float=0
    ConstraintWeights_7:float=0
    ConstraintWeights_8:float=0
   ) -> (Pose) {
    alphaBoolBlend {blendOption=Linear}
    boneToModify {boneName=dyn_beardBraid_01_r}
    constraintSetup [
     {targetBone {boneName=head_offset}}
     {targetBone {boneName=dyn_beard_01}}
     {targetBone {boneName=clavicle_r}}
     {targetBone {boneName=spine_05}}
     {targetBone {boneName=spine_04}}
     {targetBone {boneName=spine_03}}
     {targetBone {boneName=clavicle_l}}
     {targetBone {boneName=spine_02}}
     {targetBone {boneName=neck_01}}
    ]
    constraintWeights [1,1,1,1,1,1,1,1,1]
   }
   n20 = Anim Dynamics - Bone: dyn_beardBraid_02_c(ComponentPose=@n19, Alpha=0.7, ExternalForce="0,0,0") {
    angularDampingOverride=1
    angularSpringConstant=500
    bAngularSpring=true
    bGravityOverrideInSimSpace=true
    bOverrideAngularDamping=true
    bUseSphericalLimits=true
    gravityScale=0.2
    simulationSpace=RootRelative
    alphaBoolBlend {blendOption=Linear}
    boundBone {boneName=dyn_beardBraid_02_c}
    chainEnd {boneName=dyn_beardBraid_03_c}
    physicsBodyDefinitions [{
      collisionType=CustomSphere
      sphereCollisionRadius=1.25
      boundBone {boneName=dyn_beardBraid_02_c}
      boxExtents=<5 20 20>
      constraintSetup {angularLimitsMax=<0 20 30>, angularLimitsMin=<0 -20 -80>, angularTarget=<0 0 0>}
      localJointOffset=<5 0 0>
     }]
    relativeSpaceBone {boneName=head_offset}
    sphericalLimits [{limitRadius=18.65, drivingBone {boneName=spine_05}, sphereLocalOffset=<0 0 0>}]
   }
   n21 = Anim Dynamics - Bone: dyn_beardBraid_03_c(ComponentPose=@n20, Alpha=0.7, ExternalForce="0,0,0") {
    angularDampingOverride=0.9999
    angularSpringConstant=500
    bAngularSpring=true
    bGravityOverrideInSimSpace=true
    bOverrideAngularDamping=true
    bUseSphericalLimits=true
    gravityScale=0.2
    simulationSpace=RootRelative
    alphaBoolBlend {blendOption=Linear}
    boundBone {boneName=dyn_beardBraid_03_c}
    physicsBodyDefinitions [{
      collisionType=CustomSphere
      sphereCollisionRadius=1.75
      boundBone {boneName=dyn_beardBraid_03_c}
      boxExtents=<10 20 20>
      constraintSetup {angularLimitsMax=<0 20 30>, angularLimitsMin=<0 -20 -30>, angularTarget=<0 0 0>}
      localJointOffset=<7 0 0>
     }]
    relativeSpaceBone {boneName=head_offset}
    sphericalLimits [{limitRadius=18.85, drivingBone {boneName=spine_05}, sphereLocalOffset=<0 0 0>}]
   }
   n22 = Anim Dynamics - Bone: dyn_beardBraid_01_l(ComponentPose=@n21, Alpha=1, ExternalForce="0,0,0") {
    angularBiasOverride=0.99
    angularDampingOverride=0.99
    angularSpringConstant=500
    bAngularSpring=true
    bGravityOverrideInSimSpace=true
    bOverrideAngularBias=true
    bOverrideAngularDamping=true
    bUseSphericalLimits=true
    gravityScale=0.2
    simulationSpace=RootRelative
    alphaBoolBlend {blendOption=Linear}
    boundBone {boneName=dyn_beardBraid_01_l}
    chainEnd {boneName=dyn_beardBraid_03_c}
    physicsBodyDefinitions [{
      collisionType=CustomSphere
      sphereCollisionRadius=1
      boundBone {boneName=dyn_beardBraid_01_l}
      boxExtents=<5 20 20>
      constraintSetup {angularLimitsMax=<0 5 30>, angularLimitsMin=<0 -20 -10>, angularTarget=<0 0 0>}
      localJointOffset=<5 0 0>
     }]
    relativeSpaceBone {boneName=head_offset}
    sphericalLimits [{limitRadius=19, drivingBone {boneName=spine_05}, sphereLocalOffset=<0 0 0>}]
   }
   n23 = Anim Dynamics - Bone: dyn_beardBraid_02_l(ComponentPose=@n22, Alpha=1, ExternalForce="0,0,0") {
    angularDampingOverride=0.99
    angularSpringConstant=500
    bAngularSpring=true
    bGravityOverrideInSimSpace=true
    bOverrideAngularDamping=true
    bUseSphericalLimits=true
    gravityScale=0.2
    simulationSpace=RootRelative
    alphaBoolBlend {blendOption=Linear}
    boundBone {boneName=dyn_beardBraid_02_l}
    physicsBodyDefinitions [{
      collisionType=CustomSphere
      sphereCollisionRadius=1
      boundBone {boneName=dyn_beardBraid_02_l}
      boxExtents=<5 20 20>
      constraintSetup {angularLimitsMax=<0 20 30>, angularLimitsMin=<0 -20 -15>, angularTarget=<0 0 0>}
      localJointOffset=<5 0 0>
     }]
    relativeSpaceBone {boneName=head_offset}
    sphericalLimits [{limitRadius=19, drivingBone {boneName=spine_05}, sphereLocalOffset=<0 0 0>}]
   }
   n24 = Anim Dynamics - Bone: dyn_beardBraid_01_r(ComponentPose=@n23, Alpha=1, ExternalForce="0,0,0") {
    angularBiasOverride=0.99
    angularDampingOverride=0.99
    angularSpringConstant=500
    bAngularSpring=true
    bGravityOverrideInSimSpace=true
    bOverrideAngularBias=true
    bOverrideAngularDamping=true
    bUseSphericalLimits=true
    gravityScale=0.2
    simulationSpace=RootRelative
    alphaBoolBlend {blendOption=Linear}
    boundBone {boneName=dyn_beardBraid_01_r}
    chainEnd {boneName=dyn_beardBraid_03_c}
    physicsBodyDefinitions [{
      collisionType=CustomSphere
      sphereCollisionRadius=1
      boundBone {boneName=dyn_beardBraid_01_r}
      boxExtents=<5 20 20>
      constraintSetup {angularLimitsMax=<0 20 30>, angularLimitsMin=<0 -5 -10>, angularTarget=<0 0 0>}
      localJointOffset=<5 0 0>
     }]
    relativeSpaceBone {boneName=head_offset}
    sphericalLimits [{limitRadius=19, drivingBone {boneName=spine_05}, sphereLocalOffset=<0 0 0>}]
   }
   n25 = Anim Dynamics - Bone: dyn_beardBraid_02_r(ComponentPose=@n24, Alpha=1, ExternalForce="0,0,0") {
    angularDampingOverride=0.99
    angularSpringConstant=500
    bAngularSpring=true
    bGravityOverrideInSimSpace=true
    bOverrideAngularDamping=true
    bUseSphericalLimits=true
    gravityScale=0.2
    simulationSpace=RootRelative
    alphaBoolBlend {blendOption=Linear}
    boundBone {boneName=dyn_beardBraid_02_r}
    physicsBodyDefinitions [{
      collisionType=CustomSphere
      sphereCollisionRadius=1
      boundBone {boneName=dyn_beardBraid_02_r}
      boxExtents=<5 20 20>
      constraintSetup {angularLimitsMax=<0 20 30>, angularLimitsMin=<0 -20 -15>, angularTarget=<0 0 0>}
      localJointOffset=<5 0 0>
     }]
    relativeSpaceBone {boneName=head_offset}
    sphericalLimits [{limitRadius=19, drivingBone {boneName=spine_05}, sphereLocalOffset=<0 0 0>}]
   }
   n26 = Anim Dynamics - Bone: dyn_hair_bck_01(ComponentPose=@n25, Alpha=0.7, ExternalForce="0,0,0") {
    angularDampingOverride=1
    angularSpringConstant=1000
    bAngularSpring=true
    bGravityOverrideInSimSpace=true
    bOverrideAngularDamping=true
    bUseGravityOverride=true
    gravityScale=0.35
    simulationSpace=RootRelative
    alphaBoolBlend {blendOption=Linear}
    boundBone {boneName=dyn_hair_bck_01}
    componentAppliedLinearAccClamp=<3000 3000 3000>
    componentLinearAccScale=<0.5 0.5 0.5>
    componentLinearVelScale=<0.5 0.5 0.5>
    gravityOverride {z=-980}
    physicsBodyDefinitions [{
      sphereCollisionRadius=0
      boundBone {boneName=dyn_hair_bck_01}
      boxExtents=<10 15 30>
      constraintSetup {angularLimitsMax=<2 8 1>, angularLimitsMin=<-2 -8 -20>, angularTarget=<1 0 0>}
      localJointOffset=<7.5 0 0>
     }]
   }
   n27 = Anim Dynamics - Bone: dyn_hair_bck_02(ComponentPose=@n26, Alpha=0.7, ExternalForce="0,0,0") {
    angularDampingOverride=1
    angularSpringConstant=20000
    bAngularSpring=true
    bGravityOverrideInSimSpace=true
    bOverrideAngularDamping=true
    bUseGravityOverride=true
    gravityScale=0.35
    simulationSpace=RootRelative
    alphaBoolBlend {blendOption=Linear}
    boundBone {boneName=dyn_hair_bck_02}
    componentAppliedLinearAccClamp=<3000 3000 3000>
    componentLinearAccScale=<0.5 0.5 0.5>
    componentLinearVelScale=<0.5 0.5 0.5>
    gravityOverride {z=-980}
    physicsBodyDefinitions [{
      sphereCollisionRadius=0
      boundBone {boneName=dyn_hair_bck_02}
      boxExtents=<10 15 30>
      constraintSetup {angularLimitsMax=<2 8 1>, angularLimitsMin=<-2 -8 -20>, angularTarget=<1 0 0>}
      localJointOffset=<7.5 0 0>
     }]
   }
   n28 = "Transform (Modify) Bone - Bone: dyn_beardBraid_01_c"(ComponentPose=@n27, Alpha=1, Translation="0,0,0", Rotation="0.000,-20,0.000", Scale="1,1,1") {
    rotationMode=BMM_Additive
    rotationSpace=BCS_BoneSpace
    translationMode=BMM_Additive
    translationSpace=BCS_BoneSpace
    alphaBoolBlend {blendOption=Linear}
    alphaScaleBiasClamp {bInterpResult=true, interpSpeedDecreasing=15, interpSpeedIncreasing=15}
    boneToModify {boneName=dyn_beardBraid_01_c}
    rotation {yaw=-20}
   }
   n29 = "Transform (Modify) Bone - Bone: dyn_beardBraid_03_c"(ComponentPose=@n28, Alpha=1, Translation="0,0,0", Rotation="0.000,15,0.000", Scale="1,1,1") {
    rotationMode=BMM_Additive
    rotationSpace=BCS_BoneSpace
    translationMode=BMM_Additive
    translationSpace=BCS_BoneSpace
    alphaBoolBlend {blendOption=Linear}
    alphaScaleBiasClamp {bInterpResult=true, interpSpeedDecreasing=15, interpSpeedIncreasing=15}
    boneToModify {boneName=dyn_beardBraid_03_c}
    rotation {yaw=15}
   }
   Output Pose(Result=Component To Local(ComponentPose=@n29)) {layerGroup=None}
  ]}
 EventGraph {nodes [
   Event BlueprintUpdateAnimation()
   > Get!(execute:Exec) -> (then:Exec, else:Exec, SkelMesh:Skeletal Mesh Component Object Reference)
   > n1 = Sequence!(execute) -> (then_0, then_1, then_2:Exec, then_3:Exec)
   > Set look_left(look_left:double=GetAlphaFromReferenceTransformDelta(
     SkeletalMeshComponent=Get SkelMesh()
     BoneName=head
     BaseName=spine_05
     DeltaComponentType=QuaternionTwist
     InputRange=(X=0,Y=45)
     TwistAxis="1,0,0"
    )) -> (Output_Get:double)
   Set look_right(
    execute=@n1.then_1
    look_right:double=GetAlphaFromReferenceTransformDelta(
     SkeletalMeshComponent=Get SkelMesh()
     BoneName=head
     BaseName=neck_01
     DeltaComponentType=QuaternionTwist
     InputRange=(X=0,Y=-45)
     TwistAxis="1,0,0"
    )
   ) -> (Output_Get:double)
   Set look_up(
    execute=@n1.then_2
    look_up:double=GetAlphaFromReferenceTransformDelta(
     SkeletalMeshComponent=Get SkelMesh()
     BoneName=head
     BaseName=neck_01
     DeltaComponentType=QuaternionTwist
     InputRange=(X=0,Y=30)
     TwistAxis="0,0,1"
    )
   ) -> (Output_Get:double)
   Set look_down(
    execute=@n1.then_3
    look_down:double=GetAlphaFromReferenceTransformDelta(
     SkeletalMeshComponent=Get SkelMesh()
     BoneName=head
     BaseName=neck_01
     DeltaComponentType=QuaternionTwist
     InputRange=(X=0,Y=-30)
     TwistAxis="0,0,1"
    )
   ) -> (Output_Get:double)
   Event BlueprintInitializeAnimation()
   > Set SkelMesh(SkelMesh=GetPartSkeletalMeshComponent())
  ]}
]
variables [
 look_down {type=number}
 look_left {type=number}
 look_right {type=number}
 look_up {type=number}
 SkelMesh {type=SkeletalMeshComponent}
]
